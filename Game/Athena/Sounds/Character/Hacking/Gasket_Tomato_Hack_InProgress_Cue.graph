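# Sound cue for an in-progress hack, named Gasket Tomato and stored with the Athena character hacking sounds.
# A single wave player feeds the root directly and plays the Gasket_Tomato_Hack_01_longer wave, with no randomization, modulation or other nodes.
# It routes through the Fort_Player_Foley sound class with a boosted volume multiplier and a very large max distance.
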SoundCue /Game/Athena/Sounds/Character/Hacking/Gasket_Tomato_Hack_InProgress_Cue
root=WavePlayer_0
nodes [{id=WavePlayer_0, properties {soundWaveAssetPtr=&/Gasket_Tomato_Hack_01_longer}}]
properties {maxDistance=2097152, soundClassObject=&Fort_Player_Foley, volumeMultiplier=1.8667}
